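; Event-triggered AI Gallery gameplay ability that plays a conversion reaction on the avatar AI pawn, then triggers its conversion.
; It commits, stores the event's OptionalObject as the converter pawn, and modifies the current weapon's ammo when one is valid.
; After a short delay it starts a montage and waits a scalable cheer duration rather than montage completion.
; It then calls TriggerConversion with the converter and AI pawn and ends the ability.
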(event Ability|EventActivateAbilityFromEvent
  (params (EventData :type Gameplay%20Event%20Data%20Structure%20%28by%20ref%29))
  (bind (_eventtag _instigator _target _optionalobject) (Utilities|Struct|BreakGameplayEventData EventData))
  (bind _returnvalue (CallFunction|GetAvatarasFortPlayerPawn))
  (bind (_montageinfo_montagetoplay _montageinfo_animplayrate _montageinfo_animrootmotiontranslationscale) (Utilities|Struct|BreakFortGameplayAbilityMontageInfo (Variables|GetMontageInfo_MontageToPlay)))
  (bind _converter (Variables|InternalData|GetConverter))
  (bind _currentweapon (Variables|/Script/FortniteGame.FortPawn|GetCurrentWeapon _returnvalue))
  (Ability|CommitAbility)
  (bind _asfort_pawn (Utilities|Casting|CastToFortPawn _optionalobject)
    (:then
      (Variables|InternalData|SetConverter _asfort_pawn))
    (:CastFailed
      _))
  (Variables|InternalData|SetAIPawn _returnvalue)
  (Utilities|IsValid _currentweapon
    (:Is%20Valid
      (FortWeapon|ModAmmo _currentweapon 500))
    (:Is%20Not%20Valid
      _))
  (Utilities|FlowControl|Delay 0.4)
  (Ability|Tasks|PlayMontageAndWait "None" _montageinfo_montagetoplay _montageinfo_animplayrate "None" true _montageinfo_animrootmotiontranslationscale
    (:then
      (Utilities|FlowControl|Delay (ScalableFloat|GetValueatLevel (Variables|Default|GetCheerDuration))))
    (:OnCompleted
      _)
    (:OnInterrupted
      _)
    (:OnCancelled
      _))
  (if (not (== _converter 0))
    (Convert|TriggerConversion _converter (Variables|InternalData|GetAIPawn)))
  (Ability|EndAbility))
