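; Gameplay ability for a Gallery AI bot that gives supplies, adding a one-time custom weapon gift meant as its first drop once hired.
; StartDrop triggers the custom drop if enabled and the bot lacks the weapon-check tag, which the drop then adds, and calls the parent logic.
; The drop sends an animation event, waits part of the montage, tosses a configured weapon or loot-tier pickup, handles optional marking, then ends the ability.
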(event Ability|EventActivateAbility
  (|CallParentFunction) ; This node is disabled and will not be called.\nDrag off pins to build functionality.
)

(event Ability|EventOnEndAbility
  (params (bWasCancelled :type Boolean))
  (|CallParentFunction bWasCancelled) ; This node is disabled and will not be called.\nDrag off pins to build functionality.
)

(event EventStartDrop
  (if (GameplayTags|HasMatchingGameplayTag (Variables|InternalData|GetBotPawn) (Variables|Default|GetTag_WeaponCheck))
    _
    (elif (ScalableFloat|GetValueasBool (Variables|Default|GetCanDropCustomWeapon))
      (CallFunction|DropCustomWeapon)))
  (|CallParentFunction))

(event Custom|DropCustomWeapon
  (bind (_position _direction) (CallFunction|SpawnPickupPositionandDIrection))
  (bind (_gameplayeventdata _getplaylength) (CallFunction|AnimPayload))
  (bind (_position_2 _direction_2) (CallFunction|SpawnPickupPositionandDIrection))
  ;[1 1 1] We want this AI to do a Custom WEapon Drop as the first thing he gives you once hired.
  (Ability|GameplayEffect|AddLooseGameplayTags (Variables|InternalData|GetBotPawn) (GameplayTags|MakeGameplayTagContainerfromTag (Variables|Default|GetTag_WeaponCheck)))
  (Ability|SendGameplayEventtoActor (Variables|InternalData|GetBotPawn) (Variables|InternalData|GetAnimTag) _gameplayeventdata)
  (Utilities|FlowControl|Delay (/ _getplaylength (ScalableFloat|GetValueatLevel (Variables|Pickups|GetMontagePlayLengthDivisionToSpawnPickup))))
  (if (Variables|Default|GetUseWIDInsteadOfLootTier)
    (bind _returnvalue (Inventory|SpawnPickupInWorld (Variables|Default|GetCustomWeaponToDrop) 1 _position_2 _direction_2 0 false false false -1 "Tossed" "NPCGift" 0 false))
    (else
      (bind _returnvalue_2 (Inventory|SpawnPickupInWorldWithLootTier (Variables|Default|GetSpecialDropLootTierName) _position 0 0 true true _direction "Tossed" "NPCGift"))
      (if (> (Utilities|Array|Length _returnvalue_2) 0)
        _)))
  (Class|FortPickup|AddSpawnSourceTags _returnvalue (GameplayTags|GetOwnedGameplayTags (Variables|InternalData|GetBotPawn) :!class /Script/GameplayTags.GameplayTagAssetInterface))
  (if (== (ScalableFloat|GetValueatLevel (Variables|Pickups|GetEnablePickupMarking)) (Variables|InternalData|GetDefaultPickupMarkingValueForActivationCheck))
    (Utilities|FlowControl|Delay (ScalableFloat|GetValueatLevel (Variables|Pickups|GetDelayBeforeMarkingItem))) ; Waiting for pickup to be in place before marking
    (MarkerComponent|MarkActoronServer (Variables|Default|GetMarkerComponent) 0 true true)
    (else
      _))
  (CallFunction|EndGiveSupplyAbility))
