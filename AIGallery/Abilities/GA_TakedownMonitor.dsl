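; Gameplay ability for non-participant AI pawns that triggers a cheer when the pawn's converter takes down another player pawn.
; On activation it caches the avatar as the AI pawn and binds to the OnDidDamage event of the converter pawn from its Convert component.
; When the damaged player pawn's health is zero or below, it sends the Cheer gameplay event tag to the AI pawn.
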(event Ability|EventActivateAbility
  (Ability|CommitAbility)
  (bind _outfortplayerpawn (CallFunction|GetAvatarasValidFortPlayerPawn)
    (:ValidFortPlayerPawn
      (bind _output_get (Variables|InternalData|SetAIPawn _outfortplayerpawn))
      (bind _asbp_player_pawn_non_participant (Utilities|Casting|CastToBP_PlayerPawn_NonParticipant _output_get)
        (:then
          (Utilities|IsValid (Convert|GetConverterPawn (Variables|/Script/FortniteGame.FortPlayerPawn|GetConvertComponent _asbp_player_pawn_non_participant))
            (:Is%20Valid
              (EventDispatchers|BindEventtoOnDidDamage (Convert|GetConverterPawn (Variables|/Script/FortniteGame.FortPlayerPawn|GetConvertComponent _asbp_player_pawn_non_participant)) (EventDispatchers|CreateEvent :!event Custom|HireeDidDamage)))
            (:Is%20Not%20Valid
              _)))
        (:CastFailed
          _))
      (Ability|EndAbility))
    (:AvatarCastFailed
      _)))

(event Ability|EventOnEndAbility
  (params (bWasCancelled :type Boolean)))

(event Ability|EventFailedToActivatePassiveAbility
  (params (ActorInfo :type Gameplay%20Ability%20Actor%20Info%20Structure)))

(event Custom|HireeDidDamage
  (params (DamagedActor :type Actor%20Object%20Reference) (Damage :type Float%20%28single-precision%29) (InstigatedBy :type Controller%20Object%20Reference) (DamageCauser :type Actor%20Object%20Reference) (HitLocation :type Vector) (FHitComponent :type Primitive%20Component%20Object%20Reference) (BoneName :type Name) (Momentum :type Vector))
  ;[1 1 1] When Hiree has Eliminated another Fort Pawn
  (bind _asfort_player_pawn (Utilities|Casting|CastToFortPlayerPawn DamagedActor))
  (if (<= (FortPawn|GetHealth _asfort_player_pawn) 0)
    (Ability|SendGameplayEventtoActor (Variables|InternalData|GetAIPawn) (Variables|Default|GetEventTag_Cheer))))
